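# Translucent, unlit, two-sided particle material for the melee swipe effect of the Technician's Kinetic ability.
# Particle color tints the swipe texture through an emissive adjustment function.
# For opacity, a V-gradient band, shifted by dynamic PanMask, masks the swipe alpha, and LerpDetails blends between cartoon-flame breakup and the plain mask.
# Flame UVs scroll by SpeedX/SpeedY plus dynamic UVoffset, and particle alpha and depth fade scale the result.
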Material /Game/CharClasses/Technician/Traits/AppliedForce/Abilities/Kinetic/Effects/M_Kinetic_Melee_Swipe
graph {
 properties {
  bEnableResponsiveAA=true
  bUsedWithMeshParticles=true
  blendMode=BLEND_Translucent
  pixelDepthOffsetMode=PDOM_Legacy
  shadingModel=MSM_Unlit
  twoSided=true
  usedShadingModels=Unlit
 }
 nodes [
  n1 = ParticleColor()
  n2 = TextureSample() {texture=/Game/Effects/Fort_Effects/Textures/Tile/T_Melee_Swipe}
  MaterialFunctionCall(Emissive=Multiply(A=@n1.RGB, B=@n2.RGB)) {materialFunction=&MF_FX_EmissiveAdjustments}
  n3 = DynamicParameter() {desc=0 to -.5 for contrast control, defaultValue=<rgba 0 0 0 1>, paramNames [PanMask,LerpDetails,UVoffset,Param4]}
  n4 = MaterialFunctionCall(UV Channel=Add(A=TextureCoordinate(), B=@n3.PanMask) {constB=0, desc=0 to -1 to pan Mask}) {materialFunction=&LinearGradient}
  n5 = LinearInterpolate(Alpha=Multiply(A=OneMinus(None=@n4.VGradient), B=@n4.VGradient)) {constB=4}
  n6 = TextureCoordinate() {uTiling=0.5, vTiling=0.5}
  n7 = ComponentMask(None=@n6) {r=true}
  n8 = DynamicParameter() {desc=0 to -.5 for contrast control, defaultValue=<rgba 0 0 0 1>, paramNames [PanMask,LerpDetails,UVoffset,Param4]}
  n9 = Multiply(A=Time(), B=ScalarParameter() {parameterName=SpeedX})
  n10 = ComponentMask(None=@n6) {g=true}
  n11 = Multiply(A=Time(), B=ScalarParameter() {defaultValue=-0.5, parameterName=SpeedY}) {constB=0}
  n12 = AppendVector(A=Add(A=Add(A=@n8.UVoffset, B=@n7), B=Frac(None=@n9)), B=Add(A=Add(A=@n10, B=@n8.UVoffset), B=Frac(None=@n11)))
  n13 = Multiply(A=@n2.A, B=Multiply(A=@n5, B=TextureSample(UVs=@n12) {texture=/Game/Effects/Fort_Effects/Textures/Tile/T_Tile_Texture_Cartoon_Flames}.RGB))
  n14 = Multiply(A=@n2.A, B=@n5)
  n15 = DynamicParameter() {desc=0 to -.5 for contrast control, defaultValue=<rgba 0 0 0 1>, paramNames [PanMask,LerpDetails,UVoffset,Param4]}
  Multiply(A=Multiply(A=@n1.A, B=LinearInterpolate(A=@n13, B=@n14, Alpha=@n15.LerpDetails)), B=DepthFade() {fadeDistanceDefault=60})
 ]
}
