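# Unlit, two-sided alpha-composite mesh-particle material for the wave ring effect of the Technician's Kinetic ability.
# The 0-1 Over Life dynamic parameter slides an inverted, normal-map-distorted gradient across the UVs, forming a base band and sharp front edge.
# Panning, clipped electric noise blends particle color toward HotColor, then front-edge glow and an HDR remap follow.
# Band and particle alpha scale emissive and drive opacity, depth-faded except on mobile.
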Material /Game/CharClasses/Technician/Traits/AppliedForce/Abilities/Kinetic/Effects/M_Kinetic_Wave_Ring_Less
graph {
 properties {
  bUsedWithMeshParticles=true
  bUsedWithNiagaraMeshParticles=true
  blendMode=BLEND_AlphaComposite
  opacityMaskClipValue=0.25
  pixelDepthOffsetMode=PDOM_Legacy
  shadingModel=MSM_Unlit
  translucencyLightingMode=TLM_VolumetricDirectional
  twoSided=true
  usedShadingModels=Unlit
 }
 nodes [
  n1 = DynamicParameter() {paramNames [RingPan, 0-1 Over Life, Param3, Param4]}
  n2 = ScalarParameter() {defaultValue=-0.9, parameterName=OffsetAll}
  n3 = Multiply(A=TextureCoordinate(), B=ComponentMask(None=VectorParameter() {parameterName=DistortionScale}.RGB) {g=true, r=true})
  n4 = ComponentMask(None=VectorParameter() {parameterName=DistortionPanSpeed}.RGB) {g=true, r=true}
  n5 = ComponentMask(None=TextureSample(UVs=Panner(Coordinate=@n3, Speed=@n4)) {samplerType=SAMPLERTYPE_Normal, texture=/Game/Global/Textures/Default/Tileable_Patterns/2ndInfectedMapNormal}.RGB) {g=true, r=true}
  n6 = Add(A=Multiply(A=@n5, B=ScalarParameter() {defaultValue=0.02, parameterName=RingDistortionAmount}), B=TextureCoordinate())
  n7 = TextureSampleParameter2D(UVs=Add(A=Add(A=@n1.0-1 Over Life, B=@n2), B=OneMinus(None=@n6))) {desc=Small, parameterName=OpacityMask, texture=/Game/Global/Textures/Default/Masks_and_Gradients/T_Vert_Gradient_Soft}
  n8 = Power(Base=@n7.G, Exp=ScalarParameter() {defaultValue=300, parameterName=FrontEdgeExp}) {constExponent=1000}
  n9 = VectorParameter() {parameterName=HotColor, defaultValue=<rgba 0.0082 1 0.4045 0>}
  n10 = ParticleColor()
  n11 = Multiply(A=@n5, B=ScalarParameter() {defaultValue=0.02, parameterName=OverallDistortionAmount})
  n12 = Multiply(
   A=TextureCoordinate()
   B=AppendVector(A=ScalarParameter() {defaultValue=0.5, parameterName=WorldTexU}, B=ScalarParameter() {defaultValue=0.5, parameterName=WorldTexV})
  )
  n13 = VectorParameter() {parameterName=NoisePanSpeed}
  n14 = TextureSampleParameter2D(UVs=Panner(Coordinate=Add(A=@n11, B=@n12), Speed=ComponentMask(None=@n13.RGB) {g=true, r=true})) {
   parameterName=Noise
   samplerType=SAMPLERTYPE_LinearColor
   texture=/Game/Characters/Player/Male/Medium/Bodies/M_MED_SRDrift/Textures/T_ElectricNoise
  }
  n15 = ScalarParameter() {defaultValue=0.1, desc=A2B, parameterName=ClipRange}
  n16 = LinearInterpolate(
   A=Multiply(A=@n15) {constB=-1}
   Alpha=OneMinus(None=ScalarParameter() {defaultValue=0.8, desc=A2B, parameterName=ClipAmnt}) {desc=Inverse Particle Alpha}
  )
  n17 = Saturate(None=Divide(A=Subtract(A=@n14.G, B=@n16) {desc=Mask}, B=@n15))
  n18 = ScalarParameter() {defaultValue=1.5, parameterName=ColorMapExp}
  n19 = Add(
   A=Multiply(A=Multiply(A=@n8, B=@n9.RGB), B=ScalarParameter() {parameterName=FrontEdgeMult})
   B=LinearInterpolate(A=@n10.RGB, B=@n9.RGB, Alpha=Power(Base=@n17, Exp=@n18))
  )
  n20 = Saturate(None=@n19)
  n21 = ScalarParameter() {defaultValue=4, parameterName=HDRIntensity}
  n22 = ScalarParameter() {defaultValue=1, parameterName=BaseBandExp}
  n23 = Multiply(
   A=@n10.A
   B=Multiply(A=Power(Base=@n7.G, Exp=@n22) {constExponent=1000}, B=ScalarParameter() {defaultValue=0.15, parameterName=BaseBandMult})
  )
  MaterialFunctionCall(Emissive=Multiply(
    A=MaterialFunctionCall(LDR Input=@n20, HDR Tint=@n19, HDR Intensity Multiplier=@n21) {materialFunction=&DeriveHDRfromLDR}.RemappedRange
    B=@n23
   )) {materialFunction=&MF_FX_EmissiveAdjustments}
  ScalarParameter() {defaultValue=0.5, parameterName=OverallDistortionScale}
  n24 = DepthFade(Opacity=@n23, FadeDistance=ScalarParameter() {parameterName=DepthFade_Distance})
  ShadingPathSwitch(Default=@n24, Deferred=@n24, Mobile=@n23) {desc=Migrated from FeatureLevelSwitch}
 ]
}
