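# Unlit translucent particle material for melee impact effects in the Technician's Applied Force Kinetic ability, usable on Niagara sprites and mesh particles.
# An ImpactTexture parameter, defaulting to a shared melee impact texture, is multiplied by particle color.
# The RGB product feeds the shared MF_FX_EmissiveAdjustments function, while the alpha product is computed separately, so particles drive both tint and fade.
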Material /Game/CharClasses/Technician/Traits/AppliedForce/Abilities/Kinetic/Effects/M_Melee_Impact
graph {
 properties {
  bUsedWithMeshParticles=true
  bUsedWithNiagaraSprites=true
  blendMode=BLEND_Translucent
  pixelDepthOffsetMode=PDOM_Legacy
  shadingModel=MSM_Unlit
  usedShadingModels=Unlit
 }
 nodes [
  n1 = ParticleColor()
  n2 = TextureSampleParameter2D() {parameterName=ImpactTexture, texture=/Game/Effects/Fort_Effects/Textures/Melee/T_Melee_Impact_01}
  MaterialFunctionCall(Emissive=Multiply(A=@n1.RGB, B=@n2.RGB)) {materialFunction=&MF_FX_EmissiveAdjustments}
  Multiply(A=@n1.A, B=@n2.A)
 ]
}
